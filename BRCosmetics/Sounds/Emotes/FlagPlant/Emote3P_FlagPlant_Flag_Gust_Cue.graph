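# Third-person Sound Cue for flag gust sounds in the FlagPlant emote.
# A Random node picks among four gust wave variations with equal weight.
# It uses close-prop attenuation, emote foley concurrency and the emote music sound class, and keeps playing when silent.
# It also declares a send to the music audio-analysis submix.
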SoundCue /BRCosmetics/Sounds/Emotes/FlagPlant/Emote3P_FlagPlant_Flag_Gust_Cue
root=Random_1
nodes [
 {id=Random_1, children [WavePlayer_1,WavePlayer_2,WavePlayer_3,WavePlayer_4], properties {weights [1,1,1,1]}}
 {id=WavePlayer_1, properties {soundWaveAssetPtr=&/Emote_FlagPlant_Flag_Gusts_01}}
 {id=WavePlayer_2, properties {soundWaveAssetPtr=&/Emote_FlagPlant_Flag_Gusts_02}}
 {id=WavePlayer_3, properties {soundWaveAssetPtr=&/Emote_FlagPlant_Flag_Gusts_03}}
 {id=WavePlayer_4, properties {soundWaveAssetPtr=&/Emote_FlagPlant_Flag_Gusts_04}}
]
properties {
 attenuationSettings=/Game/Sounds/Attenuation/Fort_Prop_Close_Att
 maxDistance=1800
 soundClassObject=&Fort_Music_Emote
 virtualizationMode=PlayWhenSilent
 volumeMultiplier=1
 concurrencySet [/Game/Sounds/Concurrency_Rules/Athena_EmoteFoley_Conc]
 soundSubmixSends [{minSendLevel=0, soundSubmix=&Fortnite_AudioAnalysis_Music}]
}
