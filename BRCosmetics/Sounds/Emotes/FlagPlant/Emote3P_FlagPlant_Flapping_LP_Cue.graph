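# Looping third-person sound cue for the Flag Plant emote, playing a flapping flag with wind gusts and pole rattles.
# A mixer layers a looping flapping bed with two quieter looping branches, each picking a random one-shot after a random delay: pole rattles, and less frequent flag gusts.
# It uses close prop attenuation with emote music class and concurrency, sends to a music analysis submix, and keeps playing when silent.
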SoundCue /BRCosmetics/Sounds/Emotes/FlagPlant/Emote3P_FlagPlant_Flapping_LP_Cue
root=Mixer_0
nodes [
 {id=Delay_0, children [Random_1], properties {delayMax=1, delayMin=0.75}}
 {id=Delay_1, children [Random_0], properties {delayMax=1.5, delayMin=1}}
 {id=Looping_0, children [Delay_0]}
 {id=Looping_1, children [Delay_1]}
 {id=Mixer_0, children [WavePlayer_1,Looping_0,Looping_1], properties {inputVolume [1.5,0.1,0.25]}}
 {id=Random_0, children [WavePlayer_0,WavePlayer_6,WavePlayer_7,WavePlayer_8], properties {weights [1,1,1,1]}}
 {id=Random_1, children [WavePlayer_2,WavePlayer_3,WavePlayer_4,WavePlayer_5], properties {weights [1,1,1,1]}}
 {id=WavePlayer_0, properties {soundWaveAssetPtr=&/Emote_FlagPlant_Flag_Gusts_01}}
 {id=WavePlayer_1, properties {bLooping=true, soundWaveAssetPtr=&/Emote_FlagPlant_Flapping_LP}}
 {id=WavePlayer_2, properties {soundWaveAssetPtr=&/Emote_FlagPlant_Pole_Rattles_01}}
 {id=WavePlayer_3, properties {soundWaveAssetPtr=&/Emote_FlagPlant_Pole_Rattles_02}}
 {id=WavePlayer_4, properties {soundWaveAssetPtr=&/Emote_FlagPlant_Pole_Rattles_03}}
 {id=WavePlayer_5, properties {soundWaveAssetPtr=&/Emote_FlagPlant_Pole_Rattles_04}}
 {id=WavePlayer_6, properties {soundWaveAssetPtr=&/Emote_FlagPlant_Flag_Gusts_02}}
 {id=WavePlayer_7, properties {soundWaveAssetPtr=&/Emote_FlagPlant_Flag_Gusts_03}}
 {id=WavePlayer_8, properties {soundWaveAssetPtr=&/Emote_FlagPlant_Flag_Gusts_04}}
]
properties {
 attenuationSettings=/Game/Sounds/Attenuation/Fort_Prop_Close_Att
 maxDistance=1800
 soundClassObject=&Fort_Music_Emote
 virtualizationMode=PlayWhenSilent
 volumeMultiplier=1.5
 concurrencySet [/Game/Sounds/Concurrency_Rules/Athena_EmoteMusic_Conc]
 soundSubmixSends [{minSendLevel=0, soundSubmix=&Fortnite_AudioAnalysis_Music}]
}
